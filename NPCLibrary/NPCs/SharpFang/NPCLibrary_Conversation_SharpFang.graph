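# Conversation database for Sharp Fang, a medic specialist NPC, defining the greeting players get when they talk to them.
# The entry point feeds one speech node that picks contextual lines by requirement: first-ever conversation, per-match conversation count, or the player's outfit.
# First-meeting and outfit-reaction lines take priority from named curve rows.
# The speech node branches through link nodes to the shared NPC quests and basic services conversations and Sharp Fang's own Dialogue entry.
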ConversationDatabase /NPCLibrary/NPCs/SharpFang/NPCLibrary_Conversation_SharpFang
entry_points [{tag=Conversation.Entry.SharpFang, destinations [#1]}]
nodes [
 ConversationEntryPointNode_7 {
  id=#1
  class=ConversationEntryPointNode
  outputs [#2]
  properties {entryTag {tagName=Conversation.Entry.SharpFang}}
 }
 Speech_C_4 {
  id=#2
  class=Speech_C
  outputs [#3, #4, #5]
  properties {
   nodeName=""
   generalConfig {
    defaultMessage="Keep your eyes peeled for... creatures of the night!!"
    contextualMessages [
     {
      message="Heya! How can the High Stakes Club help you out?"
      contextRequirements [{requirement {class=FortControllerRequirement_IsFirstEverConversationWithNPC}}]
      priority {curve {rowName=Tandem.Speech.Priority.FirstEverInteraction}}
     }
     {
      message="Keep your eyes peeled for... creatures of the night!!"
      contextRequirements [{requirement {
         class=FortControllerRequirement_CheckNPCMatchConversationCount
         properties {nPCUniqueIdTag {tagName=AISpawnerData.Type.Specialist.Medic.SharpFang}}
        }}]
      priority {curve {rowName=None}}
     }
     {
      message="Watch out for demons out there!"
      contextRequirements [{requirement {
         class=FortControllerRequirement_CheckNPCMatchConversationCount
         properties {nPCUniqueIdTag {tagName=AISpawnerData.Type.Specialist.Medic.SharpFang}}
        }}]
      priority {curve {rowName=None}}
     }
     {
      message="I'm used to vampires and monsters, but demons?? That's new."
      contextRequirements [{requirement {
         class=FortControllerRequirement_CheckNPCMatchConversationCount
         properties {nPCUniqueIdTag {tagName=AISpawnerData.Type.Specialist.Medic.SharpFang}}
        }}]
      priority {curve {rowName=None}}
     }
     {
      message="Stake holstered and eyes on a swivel!"
      contextRequirements [{requirement {
         class=FortControllerRequirement_CheckNPCMatchConversationCount
         properties {nPCUniqueIdTag {tagName=AISpawnerData.Type.Specialist.Medic.SharpFang}}
        }}]
      priority {curve {rowName=None}}
     }
     {
      message="Demons, sprites, magic... it's just like in the old stories."
      contextRequirements [{requirement {
         class=FortControllerRequirement_CheckNPCMatchConversationCount
         properties {nPCUniqueIdTag {tagName=AISpawnerData.Type.Specialist.Medic.SharpFang}}
        }}]
      priority {curve {rowName=None}}
     }
     {
      message="I can't believe we're seeing actual, real magic!!"
      contextRequirements [{requirement {
         class=FortControllerRequirement_CheckNPCMatchConversationCount
         properties {nPCUniqueIdTag {tagName=AISpawnerData.Type.Specialist.Medic.SharpFang}}
        }}]
      priority {curve {rowName=None}}
     }
     {
      message=We're surrounded by actual magic. How cool is that??
      contextRequirements [{requirement {
         class=FortControllerRequirement_HasCID
         properties {allowedCIds [/BRCosmetics/Athena/Items/Cosmetics/Characters/Character_Silencer, /BRCosmetics/Athena/Items/Cosmetics/Characters/Character_ProdigyFire]}
        }}]
      priority {curve {rowName=Tandem.Speech.Priority.OutfitReaction}}
     }
     {
      message="Remember to take care of yourself too, Lucien."
      contextRequirements [{requirement {
         class=FortControllerRequirement_HasCID
         properties {allowedCIds [/BRCosmetics/Athena/Items/Cosmetics/Characters/Character_Sparrow, /BRCosmetics/Athena/Items/Cosmetics/Characters/Character_ProdigyHaughty]}
        }}]
      priority {curve {rowName=Tandem.Speech.Priority.OutfitReaction}}
     }
     {
      message=High Stakes Club is happy to help with your... artifact search?
      contextRequirements [{requirement {class=FortControllerRequirement_HasCID, properties {allowedCIds [&Character_IronLilac_Sly]}}}]
      priority {curve {rowName=Tandem.Speech.Priority.OutfitReaction}}
     }
     {
      message=Joni seems to trust you... but only a little. I wonder why.
      contextRequirements [{requirement {class=FortControllerRequirement_HasCID, properties {allowedCIds [/BRCosmetics/Athena/Items/Cosmetics/Characters/Character_LastHour]}}}]
      priority {curve {rowName=Tandem.Speech.Priority.OutfitReaction}}
     }
    ]
   }
  }
 }
 ConversationLinkNode_2 {id=#5, class=ConversationLinkNode, properties {remoteEntryTag {tagName=Conversation.Entry.SharpFang.Dialogue}}}
 ConversationLinkNode_0 {id=#4, class=ConversationLinkNode, properties {remoteEntryTag {tagName=Conversation.Entry.BasicServices}}}
 ConversationLinkNode_1 {id=#3, class=ConversationLinkNode, properties {remoteEntryTag {tagName=Conversation.Entry.NPCLibrary.Quests}}}
]
